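; Player-pawn Blueprint for the WingBath Medic specialist, a hireable NPC that adds character-specific speech bubbles and hire audio.
; On hire audio, it sets the bank on its hire-change component, and on conversion it shows a hire line.
; Conversion also starts an EventProcessorSubsystem filtered listener on the hirer, whose matches trigger an enemy-eliminated line.
; Ping commands share one response line with no per-command logic, and server-side death shows a death line.
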(event EventBeginPlay
  (|CallParentFunction) ; This node is disabled and will not be called.\nDrag off pins to build functionality.
)

(event Collision|EventActorBeginOverlap
  (params (OtherActor :type Actor%20Object%20Reference))
  (|CallParentFunction OtherActor) ; This node is disabled and will not be called.\nDrag off pins to build functionality.
)

(event EventTick
  (params (DeltaSeconds :type Float%20%28single-precision%29))
  (|CallParentFunction DeltaSeconds) ; This node is disabled and will not be called.\nDrag off pins to build functionality.
)

(event EventEventOnHiredAudio
  (params (Instigator :type Fort%20Pawn%20Object%20Reference))
  (Class|BPJethroPawnHireChange|SetBank (Variables|Default|GetBP_JethroPawnHireChange))
  (|CallParentFunction Instigator))

(event Conversion|EventHandlePawnConverted
  (params (InstigatorPawn :type Fort%20Pawn%20Object%20Reference) (ConvertedPawn :type Fort%20Pawn%20Object%20Reference))
  (|CallParentFunction InstigatorPawn ConvertedPawn)
  (Class|BPFLAICommonFunctions|PlaySpeechBubble ConvertedPawn "(TagName=\"Jethro.Conversation.Entry.WingBath.OnHire\")")
  (CallFunction|OnPawnEliminated InstigatorPawn))

(event OnPingCommandEvent(ConvertComponent)
  (params (CommandType :type PingCommandType%20Enum :default "GoTo"))
  (switch Utilities|FlowControl|Switch|SwitchonPingCommandType CommandType
    (:GoTo
      _)
    (:BackToMe
      _)
    (:BackToDefault
      _)
    (:HoldPosition
      _)
    (:Attack
      _))
  (Class|BPFLAICommonFunctions|PlaySpeechBubble self "(TagName=\"Jethro.Conversation.Entry.WingBath.OnPing\")"))

(event Custom|OnPawnEliminated
  (params (Converter :type Pawn%20Object%20Reference))
  (EventProcessorSubsystem|StartListeningToFilteredEvent 0 Converter (Variables|Default|GetVerb%20Filter) false (EventDispatchers|CreateEvent :!event Custom|OnFilterPassed)))

(event Custom|OnFilterPassed
  (Class|BPFLAICommonFunctions|PlaySpeechBubble self "(TagName=\"Jethro.Conversation.Entry.WingBath.EnemyEliminated\")" false))

(event EventOnDeathServer
  (params (Damage :type Float%20%28single-precision%29) (DamageTags :type Gameplay%20Tag%20Container%20Structure%20%28by%20ref%29) (Momentum :type Vector :default "0, 0, 0") (HitInfo :type Hit%20Result%20Structure%20%28by%20ref%29) (InstigatedBy :type Controller%20Object%20Reference) (DamageCauser :type Actor%20Object%20Reference) (EffectContext :type Gameplay%20Effect%20Context%20Handle%20Structure))
  (Class|BPFLAICommonFunctions|PlaySpeechBubble self "(TagName=\"Jethro.Conversation.Entry.WingBath.OnDeath\")"))

(fn ConstructionScript
  (|CallParentFunction))
