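; Gameplay ability for the WingBath Medic specialist NPC that heals the blackboard target with a Chug Splash and plays a heal speech bubble.
; Activation applies an owner cooldown that lengthens with the pawn's HealsInStorm count, then validates the target and bot controller.
; A recursive Heal event equips the item and pulses secondary fire up to a limit, then updates HealsInStorm from the target's safe-zone tag.
; Ending re-equips the best weapon.
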(event Ability|EventActivateAbility
  (bind (_targetpawn _isnotvalid) (CallFunction|GetTargetPawnfromBlackboard))
  (bind _returnvalue (Ability|MakeOutgoingGameplayEffectSpec "/NPCSpecialists/NPCs/Medic/Abilities/GE_HealCooldown.GE_HealCooldown_C"))
  (bind _healcooldown_hf (Variables|Default|GetHealCooldown_Hf))
  (bind _returnvalue_2 (ScalableFloat|GetValueatLevel _healcooldown_hf 1.0))
  (Ability|CommitAbility)
  (bind _asbp_player_pawn_sp_medic (Utilities|Casting|CastToBP_PlayerPawn_SP_Medic (Ability|GetAvatarActorfromActorInfo))
    (:then
      (Variables|InternalData|SetAIPawn _asbp_player_pawn_sp_medic)
      (bind _returnvalue_3 (Ability|GameplayEffect|AssignTagSetbyCallerMagnitude _returnvalue "(TagName=\"SetByCaller.Cooldown\")" (+ _returnvalue_2 (* (ScalableFloat|GetValueatLevel (Variables|Default|GetHealAdditionalStormCd_Hf) 1.0) (Variables|/NPCSpecialists/NPCs/Medic/Pawn/BP_PlayerPawn_SP_Medic.BP_PlayerPawn_SP_Medic_C|GetHealsInStorm (Variables|InternalData|GetAIPawn))))))
      (Ability|ApplyGameplayEffectSpecToOwner _returnvalue_3)
      (if _isnotvalid
        (Ability|EndAbility)
        (else
          (Variables|InternalData|SetTargetPawn _targetpawn)
          (bind _asfort_athena_aibot_controller (Utilities|Casting|CastToFortAthenaAIBotController (AI|GetAIController (Ability|GetAvatarActorfromActorInfo)))
            (:then
              (Variables|InternalData|SetBotController _asfort_athena_aibot_controller)
              (Variables|InternalData|SetHealCounter 0)
              (CallFunction|Heal)
              (bind _asbp_player_pawn_sp_medic_wing_bath (Utilities|Casting|CastToBP_PlayerPawn_SP_Medic_WingBath _asbp_player_pawn_sp_medic))
              (Class|BPFLAICommonFunctions|PlaySpeechBubble _asbp_player_pawn_sp_medic_wing_bath "(TagName=\"Jethro.Conversation.Entry.WingBath.OnHeal\")"))
            (:CastFailed
              (Ability|EndAbility))))))
    (:CastFailed
      ; Wrong Pawn Type - Try healing again later
      (bind _returnvalue_4 (Ability|GameplayEffect|AssignTagSetbyCallerMagnitude _returnvalue "(TagName=\"SetByCaller.Cooldown\")" _returnvalue_2))
      (Ability|ApplyGameplayEffectSpecToOwner _returnvalue_4)
      (Fortnite|FortShippingLog (Ability|GetAvatarActorfromActorInfo) "Medic Heal Ability expects a Pawn of type PlayerPawn_SP_Medic, and the AI using this ability does not inherit from it. Make sure the pawn inherits from the proper pawn." true)
      (Ability|EndAbility))))

(event Custom|Heal
  (bind _aipawn (Variables|InternalData|GetAIPawn))
  (bind _returnvalue (Class|FortAthenaAIBotController|EquipWeaponbyTag (Variables|InternalData|GetBotController) (Variables|Default|GetChugSplashWeaponTag)))
  (if (Utilities|IsValid _returnvalue :!class /Script/Engine.KismetSystemLibrary)
    (Utilities|FlowControl|Delay (ScalableFloat|GetValueatLevel (Variables|Default|GetDelayBeforeHealingAgain)))
    (Class|FortAthenaAIBotController|StartSecondaryFire (Variables|InternalData|GetBotController))
    (Utilities|FlowControl|Delay (ScalableFloat|GetValueatLevel (Variables|Default|GetDelayBetweenStartAndStopFire)))
    (Class|FortAthenaAIBotController|StopSecondaryFire (Variables|InternalData|GetBotController) true)
    (Variables|InternalData|SetHealCounter (+ (Variables|InternalData|GetHealCounter) 1))
    (if (>= (Variables|InternalData|GetHealCounter) (ScalableFloat|GetValueasInteger (Variables|Default|GetHealCountLimit)))
      (Utilities|FlowControl|Delay 0.3)
      (if (GameplayTags|HasMatchingGameplayTag (Variables|InternalData|GetTargetPawn) (Variables|Default|GetOutsideSafeZoneTag))
        (Math|Integer|IncrementInt (Variables|/NPCSpecialists/NPCs/Medic/Pawn/BP_PlayerPawn_SP_Medic.BP_PlayerPawn_SP_Medic_C|GetHealsInStorm _aipawn)
          (:%20%20
            _))
        (else
          (Variables|/NPCSpecialists/NPCs/Medic/Pawn/BP_PlayerPawn_SP_Medic.BP_PlayerPawn_SP_Medic_C|SetHealsInStorm 0 _aipawn)))
      ;[0 1 0.225] End Ability
      (Ability|EndAbility)
      (else
        (CallFunction|Heal)))
    (else
      (Ability|EndAbility))))

(event Ability|EventOnEndAbility
  (params (bWasCancelled :type Boolean))
  (|CallParentFunction bWasCancelled)
  (bind _botcontroller (Variables|InternalData|GetBotController)
    (:then
      (Class|FortAthenaAIBotController|EquipBestWeapon _botcontroller false))
    (:else)))

(fn GetTargetPawnFromBlackboard
  (bind _returnvalue (CallFunction|GetAvatarasFortPlayerPawn))
  (bind _returnvalue_2 (AI|GetAIController _returnvalue))
  (bind _returnvalue_3 (AI|GetBlackboard _returnvalue_2))
  (Utilities|IsValid _returnvalue_2
    (:Is%20Valid
      (Utilities|IsValid _returnvalue_3
        (:Is%20Valid
          (bind _asfort_pawn (Utilities|Casting|CastToFortPawn (AI|Components|Blackboard|GetValueasObject _returnvalue_3 (Variables|Default|GetBBKey_TargetActor)))
            (:then
              (return _asfort_pawn false))
            (:CastFailed
              (return 0 true))))
        (:Is%20Not%20Valid
          _)))
    (:Is%20Not%20Valid
      _))
  (return 0 true))

(fn StartSpeechBubble
  (bind _returnvalue (Ability|GetAvatarActorfromActorInfo))
  (bind _returnvalue_2 (Actor|GetComponentbyClass _returnvalue "/Script/FortniteGame.FortPawnComponent_SpeechBubble"))
  (Utilities|IsValid _returnvalue
    (:Is%20Valid
      (Utilities|IsValid _returnvalue_2
        (:Is%20Valid
          (SpeechBubble|StartSpeech _returnvalue_2 "(TokenStreamVersion=0,TagDictionary=,QueryTokenStream=(0,0),UserDescription=\"\",AutoDescription=\"undefined\")"))))))
