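# Conversation database for the WingBath Medic specialist NPC in JethroLager, opened by its WingBath Medic entry tag.
# A single speech node gives a default greeting, with priority-weighted alternatives for a first-ever conversation or for players wearing specific outfits, answered with "Su cuy'gar."
# The speech node then branches through link nodes into shared quest, Jethro services and Season 35 "Found WingBath" conversations, plus this Medic's own dialogue entry.
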ConversationDatabase /JethroLager/NPCs/Specialists/Medic/Jethro_Conversation_WingBath_Medic
entry_points [{tag=Jethro.Conversation.Entry.WingBath.Medic, destinations [#1]}]
nodes [
 ConversationEntryPointNode_7 {
  id=#1
  class=ConversationEntryPointNode
  outputs [#2]
  properties {entryTag {tagName=Jethro.Conversation.Entry.WingBath.Medic}}
 }
 Speech_C_4 {
  id=#2
  class=Speech_C
  outputs [#3, #4, #5, #6]
  properties {
   nodeName=""
   generalConfig {
    defaultMessage=Need something?
    contextualMessages [
     {
      message=What do you need?
      contextRequirements [{requirement {class=FortControllerRequirement_IsFirstEverConversationWithNPC}}]
      priority {curve {rowName=Tandem.Speech.Priority.FirstEverInteraction}}
     }
     {
      message=Su cuy'gar.
      contextRequirements [{requirement {
         class=FortControllerRequirement_HasCID
         properties {allowedCIds [
           /BRCosmetics/Athena/Items/Cosmetics/Characters/CID_960_Athena_Commando_M_Cosmos
           /BRCosmetics/Athena/Items/Cosmetics/Characters/Character_WingBath
          ]}
        }}]
      priority {curve {rowName=Tandem.Speech.Priority.OutfitReaction}}
     }
    ]
   }
  }
 }
 ConversationLinkNode_3 {
  id=#6
  class=ConversationLinkNode
  properties {remoteEntryTag {tagName=Jethro.Conversation.Entry.WingBath.Medic.Dialogue}}
 }
 ConversationLinkNode_2 {
  id=#5
  class=ConversationLinkNode
  properties {remoteEntryTag {tagName=Conversation.Entry.Season35.Found.WingBath}}
 }
 ConversationLinkNode_0 {id=#4, class=ConversationLinkNode, properties {remoteEntryTag {tagName=Conversation.Entry.JethroServices}}}
 ConversationLinkNode_1 {id=#3, class=ConversationLinkNode, properties {remoteEntryTag {tagName=Conversation.Entry.NPCLibrary.Quests}}}
]
